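# DEMO_Creative_Test is a Default Lit material in the Maze environment set, apparently a test, flagged for Nanite, skeletal mesh and mesh deformer use.
# It declares four vector parameters: one with the default name, plus Metallic, Spec and Rough.
# All default to mid grey except Metallic, which keeps its unset default.
# The graph holds no textures or math, and shows no links from the parameters to material outputs.
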Material /Game/Environments/Sets/Maze/Materials/DEMO_Creative_Test
graph {
 properties {
  bUsedWithMeshDeformer=true
  bUsedWithNanite=true
  bUsedWithSkeletalMesh=true
  pixelDepthOffsetMode=PDOM_Legacy
  usedShadingModels=Default Lit
 }
 nodes [
  VectorParameter() {defaultValue=<rgba 0.5 0.5 0.5 0>}
  VectorParameter() {parameterName=Metallic}
  VectorParameter() {parameterName=Spec, defaultValue=<rgba 0.5 0.5 0.5 0>}
  VectorParameter() {parameterName=Rough, defaultValue=<rgba 0.5 0.5 0.5 0>}
 ]
}
